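# Editor utility widget that shows one actor component's display name above a property view for it.
# SetComponent binds and reveals the Fort Quest Tool property view returned by getDetailsView; two other details views in the overlay stay collapsed.
# UpdateDetailsViewProperties filters that view by writing PropertiesToShow and CategoriesToShow through SetEditorProperty.
# ForceValidateDetailsView refreshes it by clearing the object, then setting currentComponent, which no graph here assigns.
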EditorUtilityWidgetBlueprint /Game/NeverCook/EditorUtilities/CustomCompPropView/CustomCompPropView_Details
graphs [
 EventGraph {links [], nodes []}
 ForceValidateDetailsView {nodes [
   ForceValidateDetailsView() {is=title K2Node_FunctionEntry}
   n1 = getDetailsView!() -> (view:Property View Base Object Reference)
   > SetObject(self=@n1)
   > SetObject(self=@n1, NewObject=Get currentComponent() -> (currentComponent:Actor Component Object Reference))
  ]}
 SetComponent {nodes [
   n1 = SetComponent!() -> (then, Component:Actor Component Object Reference) {is=title K2Node_FunctionEntry}
   n2 = getDetailsView()
   > SetObject(self=@n2, NewObject=@n1)
   > SetVisibility(self=@n2, InVisibility=Visible)
   > "SetText (Text)"(self=Get TextBlock_141() -> (TextBlock_141:Text Object Reference), InText="To Text (String)"(InString=GetDisplayName(Object=@n1)))
  ]}
 UpdateDetailsViewProperties {nodes [
   n1 = UpdateDetailsViewProperties!() -> (then, Properties:Array of Names, Categories:Array of Names) {is=title K2Node_FunctionEntry}
   n2 = getDetailsView()
   > SetEditorProperty(Object=@n2, PropertyName=PropertiesToShow, PropertyValue:Array of Names=@n1)
   > SetEditorProperty(Object=@n2, PropertyName=CategoriesToShow, PropertyValue=@n1.Categories)
  ]}
 getDetailsView {nodes [
   getDetailsView() {is=title K2Node_FunctionEntry}
   > Return Node!(
    execute
    view:Property View Base Object Reference=Get FortQuestToolPropertyView_120() -> (FortQuestToolPropertyView_120:Fort Quest Tool Property View Object Reference)
   )
   Get DetailsView_0() -> (DetailsView_0:Details View Object Reference)
   Get FortQuestToolDetailsView_175() -> (FortQuestToolDetailsView_175:Fort Quest Tool Details View Object Reference)
  ]}
]
variables [
 Categories {type="array<string>", default []}
 currentComponent {type=ActorComponent, default=null}
 Properties {type="array<string>", default []}
]
widgets {
 info {rootWidgetClass=&VerticalBox, widgetCount=10}
 widgets [
  VerticalBox_2 {widgetClassPath=&VerticalBox, widget=VerticalBox_2}
  Border_0 {
   widgetClassPath=&Border
   parent=VerticalBox_2
   widget=Border_0
   properties {brushColor=<rgba 0.4792 0.4792 0.4792 0.162>, padding=<ltrb 3 3 3 3>}
   slot {class=VerticalBoxSlot, properties {size {sizeRule=Fill, value=1}}}
  }
  VerticalBox_1 {widgetClassPath=&VerticalBox, parent=Border_0, widget=VerticalBox_1, slot {class=BorderSlot, properties {padding=<ltrb 3 3 3 3>}}}
  HorizontalBox_0 {widgetClassPath=&HorizontalBox, parent=VerticalBox_1, widget=HorizontalBox_0, slot {class=VerticalBoxSlot}}
  TextBlock_141 {
   widgetClassPath=&TextBlock
   bIsVariable=true
   parent=HorizontalBox_0
   widget=TextBlock_141
   properties {font {
     bForceMonospaced=false
     bMaterialIsStencil=false
     fontMaterial=null
     fontObject=&Roboto
     letterSpacing=0
     monospacedWidth=1
     size=9
     skewAmount=0
     typefaceFontName=Bold
     outlineSettings {
      bApplyOutlineToDropShadows=false
      bMiteredCorners=false
      bSeparateFillAlpha=false
      outlineMaterial=null
      outlineSize=0
      outlineColor=<rgba 0 0 0 1>
     }
    }}
   slot {class=HorizontalBoxSlot, properties {size {sizeRule=Fill, value=1}}}
  }
  Overlay_1 {widgetClassPath=&Overlay, parent=VerticalBox_1, widget=Overlay_1, slot {class=VerticalBoxSlot}}
  DetailsView_0 {
   widgetClassPath=&DetailsView
   parent=Overlay_1
   widget=DetailsView_0
   properties {visibility=Collapsed}
   slot {class=OverlaySlot, properties {horizontalAlignment=HAlign_Fill, verticalAlignment=VAlign_Fill}}
  }
  FortQuestToolDetailsView_175 {
   widgetClassPath=/Script/FortniteEditor.FortQuestToolDetailsView
   bIsVariable=true
   parent=Overlay_1
   widget=FortQuestToolDetailsView_175
   properties {visibility=Collapsed}
   slot {class=OverlaySlot, properties {horizontalAlignment=HAlign_Fill, verticalAlignment=VAlign_Fill}}
  }
  FortQuestToolPropertyView_120 {
   widgetClassPath=/Script/FortniteEditor.FortQuestToolPropertyView
   bIsVariable=true
   parent=Overlay_1
   widget=FortQuestToolPropertyView_120
   properties {visibility=Collapsed}
   slot {class=OverlaySlot, properties {horizontalAlignment=HAlign_Fill, verticalAlignment=VAlign_Fill}}
  }
  Spacer_59 {widgetClassPath=&Spacer, parent=VerticalBox_2, widget=Spacer_59, properties {size=<1 5>}, slot {class=VerticalBoxSlot}}
 ]
}
