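# Editor utility widget that shows filtered property detail panels for a chosen component class on the first selected level actor.
# On Construct, it subscribes to level selection changes if Properties or Categories is non-empty and ComponentClass is a specific subclass.
# SetActor clears the StackBox, then adds one CustomCompPropView_Details widget per matching component, passing the filters and component.
# Panels rebuild only when the actor differs from currentActor, although clearing always runs.
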EditorUtilityWidgetBlueprint /Game/NeverCook/EditorUtilities/CustomCompPropView/CustomCompPropViewWidget
graphs [
 AddDetailsForComponent {nodes [
   n1 = AddDetailsForComponent!() -> (then, Component:Actor Component Object Reference) {is=title K2Node_FunctionEntry}
   > n2 = Create Widget!(
    execute:Exec
    WorldContextObject:Object Reference=Self-Reference()
    Class:User Widget Class Reference=&/CustomCompPropView_Details.CustomCompPropView_Details_C
    OwningPlayer:Player Controller Object Reference
   ) -> (then:Exec, ReturnValue:Custom Comp Prop View Details Object Reference)
   > Set l_widget(l_widget:Custom Comp Prop View Details Object Reference=@n2) -> (Output_Get:Custom Comp Prop View Details Object Reference)
   > UpdateDetailsViewProperties!(
    execute
    self:Custom Comp Prop View Details Object Reference=Get l_widget() -> (l_widget:Custom Comp Prop View Details Object Reference)
    Properties:Array of Names=Get Properties() -> (Properties:Array of Names)
    Categories:Array of Names=Get Categories() -> (Categories:Array of Names)
   ) -> (then)
   > SetComponent!(execute, self:Custom Comp Prop View Details Object Reference=Get l_widget(), Component:Actor Component Object Reference=@n1) -> (then)
   > AddChild(self=Get StackBox_0() -> (StackBox_0:Stack Box Object Reference), Content=Get l_widget())
  ]}
 ClearComponents {nodes [
   ClearComponents() {is=title K2Node_FunctionEntry}
   > n1 = Reverse for Each Loop!(Exec:Exec, Array:Array of Widget Object References=GetAllChildren(self=Get StackBox_0())) -> (LoopBody:Exec, ArrayIndex:int, ArrayElement:Widget Object Reference, Completed:Exec) {is=title K2Node_MacroInstance}
   > n2 = Cast To CustomCompPropView_Details!(execute, Object=@n1.ArrayElement) -> (then, CastFailed, AsCustom Comp Prop View Details:Custom Comp Prop View Details Object Reference) {is=title K2Node_DynamicCast}
   > RemoveFromParent(self=@n2)
  ]}
 EventGraph {nodes [
   Event Construct()
   n1 = Length(TargetArray:Array of Names=Get Categories())
   n2 = Length(TargetArray=Get Properties())
   > Branch(Condition=OR Boolean(A="integer > integer"(A=@n2), B="integer > integer"(A=@n1)))
   n3 = Get ComponentClass() -> (ComponentClass:Actor Component Class Reference)
   > Branch(Condition=AND Boolean(
     A=IsValidClass!(Class:Object Class Reference=@n3) -> (ReturnValue:bool)
     B="Not Equal (Class)"(A:Object Class Reference=@n3, B:Object Class Reference=/Script/Engine.ActorComponent) -> (ReturnValue:bool=~) {is=title K2Node_PromotableOperator}
    ))
   > n4 = GetSelectionSet(self=Get LevelEditorSubsystem())
   > Bind Event to On Selection Change(self:Typed Element Selection Set Object Reference=@n4, Delegate=Create Event()) {is=title K2Node_AddDelegate}
   > handleLevelSelectionChanged!(execute, self, SelectionSet:Typed Element Selection Set Object Reference=@n4) -> (then)
  ]}
 SetActor {nodes [
   n1 = SetActor!() -> (then, actor:Actor Object Reference) {is=title K2Node_FunctionEntry}
   > ClearComponents() {is=title K2Node_CallFunction}
   > Branch(Condition="Not Equal (Object)"(A:Object Reference=@n1, B:Object Reference=Get currentActor() -> (currentActor:Actor Object Reference)) -> (ReturnValue:bool=~) {is=title K2Node_PromotableOperator})
   > Set currentActor(currentActor:Actor Object Reference=@n1) -> (Output_Get:Actor Object Reference)
   > n2 = Sequence!(execute) -> (then_0, then_1, then_2:Exec)
   n3 = Is Valid(exec=@n2.then_1, InputObject=Get currentActor())
   Return Node(execute=@n3.Is Not Valid)
   n4 = For Each Loop(
    Exec=@n2.then_2
    Array:Array of Actor Component Object References=Get Components By Class!(self:Actor Object Reference=Get currentActor(), ComponentClass:Actor Component Class Reference=Get ComponentClass()) -> (ReturnValue:Array of Actor Component Object References)
   ) -> (Array Element:Actor Component Object Reference)
   > AddDetailsForComponent!(execute, self, Component:Actor Component Object Reference=@n4) -> (then) {is=title K2Node_CallFunction}
  ]}
 handleLevelSelectionChanged {nodes [
   n1 = handleLevelSelectionChanged!() -> (then, SelectionSet:Typed Element Selection Set Object Reference) {is=title K2Node_FunctionEntry}
   > n2 = GetSelectedObjects!(execute, self:Typed Element Selection Set Object Reference=@n1, InRequiredClass:Object Class Reference=&Actor) -> (then, ReturnValue:Array of Object References)
   > n3 = Branch(Condition=Is Valid Index!(TargetArray:Array of Object References=@n2, IndexToTest:int) -> (ReturnValue:bool))
   > n4 = Cast To Actor!(execute, Object="Get (a copy)"(Array:Array of Object References=@n2) -> (Output:Object Reference)) -> (then, CastFailed, AsActor:Actor Object Reference) {is=title K2Node_DynamicCast}
   > SetActor!(execute, self, actor:Actor Object Reference=@n4) -> (then) {is=title K2Node_CallFunction}
   SetActor(execute=@n3.else)
  ]}
]
variables [
 Categories {type="array<string>", default []}
 ComponentClass {type=Class, default=&ActorComponent}
 currentActor {type=Actor, default=null}
 Properties {type="array<string>", default []}
]
widgets {
 info {rootWidgetClass=&StackBox, widgetCount=1}
 widgets [StackBox_0 {widgetClassPath=&StackBox, bIsVariable=true, widget=StackBox_0, properties {orientation=Orient_Vertical}}]
}
